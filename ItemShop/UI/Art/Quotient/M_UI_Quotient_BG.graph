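# UI material for an Item Shop "Quotient" background: an animated, layered backdrop with tinted base, spotlight and color overlay.
# Builds gradient-colored base, breathing spotlight, purple glow and bottom-of-screen gradient layers, composited through overlay, lighten and lerp blends, then hue-shifted.
# Flow-mapped specular texture gains an expanding shine ring through hard-light and color-burn blends inside a circular mask.
# A static switch selects viewport or UI UVs, with screen-space scaling variants.
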Material /ItemShop/UI/Art/Quotient/M_UI_Quotient_BG
graph {
 properties {bUsedWithStaticMesh=false, materialDomain=MD_UI, pixelDepthOffsetMode=PDOM_Legacy, usedShadingModels=Default Lit}
 nodes [
  n1 = NamedRerouteUsage()
  n2 = Saturate(None=LinearInterpolate(A=@n1, B=NamedRerouteUsage(), Alpha=@n1))
  n3 = NamedRerouteUsage()
  MaterialExpressionOverlay_0 = MaterialXOverlay!(A=@n3, B=NamedRerouteUsage(), Alpha)
  n4 = Add(A=Multiply(A=NamedRerouteUsage()) {constB=0.3}, B=Saturate(None=Add(A=@MaterialExpressionOverlay_0, B=@n3)))
  n5 = MaterialFunctionCall!(Blend=NamedRerouteUsage(), Base=@n4) {materialFunction=/Engine/Functions/Engine_MaterialFunctions03/Blends/Blend_Lighten}
  n6 = Multiply(
   A=ScalarParameter() {defaultValue=0.025, group=Bkgd_ColorOverlay, parameterName=Base Bkg Texture Lighten-Value}
   B=ScalarParameter() {defaultValue=0.5, group=Bkgd_ColorOverlay, parameterName=Base Bkg Texture Lighten-Multiply}
  ) {constA=0.025, constB=0.5}
  n7 = Saturate(None=Add(A=Multiply(A=@n5.Result, B=@n6) {constB=0.001}, B=Saturate(None=@n4)))
  MaterialExpressionOverlay_1 = MaterialXOverlay(A=@n7, B=NamedRerouteUsage())
  n8 = Saturate(None=Add(A=@MaterialExpressionOverlay_1, B=@n7))
  n9 = NamedRerouteUsage()
  MaterialFunctionCall(
   Hue Shift Percentage=ScalarParameter() {group=PostProcess, parameterName=HueShift, sliderMax=1}
   Texture=Saturate(None=LinearInterpolate(A=@n2, B=@n8, Alpha=@n9) {constAlpha=1})
  ) {materialFunction=&HueShift}
  n10 = VectorParameter() {group=Bkgd_BottomOfScreen, parameterName=Bottom of Screen Mask Color, defaultValue=<rgba 0.0044 0.0137 0.0168 1>}
  n11 = Add(
   A=NamedRerouteUsage()
   B=AppendVector(A=Constant(), B=ScalarParameter() {defaultValue=-0.4, group=Bkgd_BottomOfScreen, parameterName=Bottom of Screen Mask Height})
  )
  n12 = Multiply(
   A=MaterialFunctionCall(
    Alpha=MaterialFunctionCall(UV Channel=@n11) {materialFunction=&LinearGradient}.VGradient
    Min=ScalarParameter() {defaultValue=0.1, group=Bkgd_BottomOfScreen, parameterName=Bottom of Screen Mask Smooth Min}
    Max=ScalarParameter() {defaultValue=0.25, group=Bkgd_BottomOfScreen, parameterName=Bottom of Screen Mask Smooth Max}
   ) {materialFunction=&SmoothStep}.Result
   B=ScalarParameter() {defaultValue=0.75, group=Bkgd_BottomOfScreen, parameterName=Bottom of Screen Mask Alpha}
  )
  NamedRerouteDeclaration(None=LinearInterpolate(B=@n10.RGB, Alpha=@n12)) {name=Bottom of Screen Gradient, nodeColor=<rgba 0 0.5534 1 1>}
  NamedRerouteDeclaration(None=OneMinus(None=@n12)) {name=Bottom of Screen Gradient Mask, nodeColor=<rgba 0 0.6398 1 1>}
  n13 = VectorParameter() {group=Bkgd_Specular, parameterName=Organic Specular Shadows, defaultValue=<rgba 0.0804 0.0621 0.3542 1>}
  n14 = LinearInterpolate(
   A=VectorParameter() {group=Bkgd_Specular, parameterName=Organic Specular Highlight, defaultValue=<rgba 0.2705 0.6445 1 1>}.RGB
   B=@n13.RGB
   Alpha=NamedRerouteUsage()
  )
  n15 = NamedRerouteUsage()
  NamedRerouteDeclaration(None=Saturate(None=MaterialFunctionCall!(Color=@n14, Luminosity=@n15) {materialFunction=/Engine/Functions/Engine_MaterialFunctions01/ImageAdjustment/Luminosity_And_Color}.Result)) {name=Colored Organic Specular, nodeColor=<rgba 0.121 0.4041 1 1>}
  NamedRerouteDeclaration(None=TextureSample(UVs=NamedRerouteUsage()) {samplerType=SAMPLERTYPE_Alpha, texture=&/T_UI_Quotient_BkgBase_Grey_512}.RGBA) {name=Base Bkg Texture, nodeColor=<rgba 0.121 0.4041 1 1>}
  n16 = NamedRerouteUsage()
  n17 = Multiply(
   A=ScalarParameter() {defaultValue=0.85, group=Bkgd_SpotLight, parameterName=Spotlight Size, sliderMax=2, sliderMin=0.01}
   B=NamedRerouteUsage()
  )
  n18 = Constant()
  n19 = ScalarParameter() {defaultValue=0.1, group=Bkgd_SpotLight, parameterName=Spotlight Height Position}
  n20 = MaterialFunctionCall(
   UV=Add(
    A=MaterialFunctionCall(UVs=@n16, Texture Scale=@n17) {materialFunction=&ScaleUVsByCenter}.UVs
    B=AppendVector(A=@n18, B=@n19)
   )
   GradientStart=ScalarParameter() {group=Bkgd_SpotLight, parameterName=Spotlight Grad Start}
   GradientEnd=ScalarParameter() {defaultValue=0.3, group=Bkgd_SpotLight, parameterName=Spotlight Grad End}
  ) {materialFunction=&MF_UI_RadialGradient}
  n21 = Multiply(A=ScalarParameter() {defaultValue=2, group=Bkgd_SpotLight, parameterName=Spotlight Intensity}, B=NamedRerouteUsage())
  NamedRerouteDeclaration(None=LinearInterpolate(
    B=VectorParameter() {group=Bkgd_SpotLight, parameterName=Center Spotlight, defaultValue=<rgba 1 1 1 1>}.RGB
    Alpha=Multiply(A=@n20.Smooth, B=@n21)
   ) {constAlpha=0}) {name=Spotlight, nodeColor=<rgba 0.121 0.4041 1 1>}
  NamedRerouteDeclaration(None=StaticSwitchParameter(
    True=ScreenPosition().ViewportUV
    False=MaterialFunctionCall() {materialFunction=&GetUserInterfaceUV}.Normalized UV
   ) {defaultValue=true, group=UV Space, parameterName=Is3DSpace}) {name=Normal UV, nodeColor=<rgba 0 0 0 1>}
  n22 = NamedRerouteUsage()
  n23 = NamedRerouteUsage()
  n24 = NamedRerouteUsage()
  n25 = VectorParameter() {parameterName=Background Base Color Dark, defaultValue=<rgba 0.0039 0.0075 0.0152 1>}
  n26 = VectorParameter() {parameterName=Background Base Color Light, defaultValue=<rgba 0.0107 0.0482 0.0938 1>}
  n27 = NamedRerouteUsage()
  NamedRerouteDeclaration(None=StaticSwitchParameter(True=LinearInterpolate(A=@n22, B=@n23, Alpha=@n24), False=LinearInterpolate(A=@n25.RGB, B=@n26.RGB, Alpha=@n27)) {group=Bkgd_Gradient, parameterName=Use color map gradients}) {name=Colored Base Bkg Texture, nodeColor=<rgba 0.121 0.4041 1 1>}
  n28 = ScalarParameter() {group=Bkgd_Gradient, parameterName=Bkg Linear Gradient Rot}
  n29 = ScalarParameter() {group=Bkgd_Gradient, parameterName=Bkg Linear Gradient Offset}
  n30 = ScalarParameter() {group=Bkgd_Gradient, parameterName=Bkg Gradient Start}
  n31 = ScalarParameter() {defaultValue=1, group=Bkgd_Gradient, parameterName=Bkg Gradient End}
  n32 = NamedRerouteUsage()
  NamedRerouteDeclaration(None=LinearInterpolate(
    A=MaterialFunctionCall(GradientStart=@n30, GradientEnd=@n31, RotationAngle=@n28, GradientOffset=@n29) {materialFunction=&MF_UI_LinearGradient}.Result
    B=MaterialFunctionCall(UV=@n32, GradientStart=@n30, GradientEnd=@n31) {materialFunction=&MF_UI_RadialGradient}.Smooth
   ) {constAlpha=1}) {name=Bkg Gradient Mapper, nodeColor=<rgba 0.121 0.4041 1 1>}
  n33 = Panner(Coordinate=MaterialFunctionCall!(Vector or UVs=NamedRerouteUsage()) {materialFunction=/FortUILibrary/MaterialFunctions/Basics/MF_UI_CheaperVectorToRadial}.Radial Coordinates) {speedY=-0.05}
  NamedRerouteDeclaration(None=MaterialFunctionCall!(
    UVs=@n33
    Texture Object=TextureObject() {texture=/Game/Effects/Fort_Effects/Textures/T_PinkSpike_Voronoi2D_Height}
    Grayscale Flow Map=TextureSample(UVs=@n33) {samplerType=SAMPLERTYPE_LinearGrayscale, texture=/Game/Effects/Fort_Effects/Textures/T_Contrail_StinkyCloud_Smoke}.R
    Flow Direction=Constant()
    Flow Intensity
    Flow Speed=Constant() {r=0.2}
   ) {materialFunction=/FortUILibrary/MaterialFunctions/Effects/MF_UI_GrayscaleFlowMap_New}.Distorted Texture) {name=Spinning Noise, nodeColor=<rgba 0 0 0 1>}
  NamedRerouteDeclaration(None=MaterialFunctionCall!(ScreenUV=NamedRerouteUsage(), TextureSize, ScalePivotPoint) {materialFunction=/FortUILibrary/MaterialFunctions/Basics/MF_UI_ScreenSpaceUVScaling}.Scaled to Fit) {name=Normal UV Scaled to Fit, nodeColor=<rgba 0 0 0 1>}
  NamedRerouteDeclaration(None=MaterialFunctionCall!(UV=NamedRerouteUsage(), Offset=Constant2Vector() {r=-0.002}, Rotation, Scale=Constant() {r=2}, MirrorY=StaticBool() {value=true}) {materialFunction=/FortUILibrary/MaterialFunctions/Basics/MF_UI_Symmetry_2x1_DEPRECATED}.Result) {name=Normal UV 2x1 Mirror, nodeColor=<rgba 0 0 0 1>}
  NamedRerouteDeclaration(None=MaterialFunctionCall(Speed=Constant() {r=0.125}) {materialFunction=&DebugTimeSine}.Result) {name='Breathing' Anim Time, nodeColor=<rgba 0 0 0 1>}
  n34 = VectorParameter() {group=Bkgd_Gradient, parameterName=Bkg Base Color Dark Lerp 1, defaultValue=<rgba 0.004 0.0075 0.0152 1>}
  n35 = VectorParameter() {group=Bkgd_Gradient, parameterName=Bkg Base Color Dark Lerp 2, defaultValue=<rgba 0.0088 0.0152 0.0146 1>}
  n36 = NamedRerouteUsage()
  NamedRerouteDeclaration(None=Saturate(None=LinearInterpolate(A=@n34.RGB, B=@n35.RGB, Alpha=@n36))) {name=Bkg Base Dark Lerped Color, nodeColor=<rgba 1 0.8148 0.0185 1>}
  n37 = VectorParameter() {group=Bkgd_Gradient, parameterName=Bkg Base Color Light Lerp 1, defaultValue=<rgba 0.0107 0.0482 0.0938 1>}
  n38 = VectorParameter() {group=Bkgd_Gradient, parameterName=Bkg Base Color Light Lerp 2, defaultValue=<rgba 0.0191 0.0857 0.1667 1>}
  n39 = NamedRerouteUsage()
  NamedRerouteDeclaration(None=Saturate(None=LinearInterpolate(A=@n37.RGB, B=@n38.RGB, Alpha=@n39))) {name=Bkg Base Light Lerped Color, nodeColor=<rgba 1 0.8148 0.0185 1>}
  n40 = Multiply(A=Constant2Vector() {r=0.005})
  n41 = NamedRerouteUsage()
  n42 = NamedRerouteUsage()
  n43 = Add(
   A=TextureSample(UVs=@n41) {samplerType=SAMPLERTYPE_Alpha, texture=&/T_UI_Quotient_Dark_BkgSpecular_512}.R
   B=Multiply(A=MaterialFunctionCall("Value (0-1)"=@n42) {materialFunction=&MF_UI_EaseCurves}.EaseOutSine) {constB=0.1}
  )
  n44 = OneMinus(None=MaterialFunctionCall(UV Channel=@n41) {materialFunction=&LinearGradient}.UGradient)
  NamedRerouteDeclaration(None=MaterialFunctionCall(
    UVs=Panner(Coordinate=@n41, Speed=@n40) {speedX=0.005}
    Texture Object=TextureObject() {samplerType=SAMPLERTYPE_Alpha, texture=&T_UI_Quotient_Dark_BkgSpecular_512}
    Grayscale Flow Map=LinearInterpolate(A=@n43, B=@n44)
    Flow Direction=Constant() {r=-0.38}
    Flow Speed=Constant() {r=0.1}
   ) {materialFunction=&MF_UI_GrayscaleFlowMap_New}.Distorted Texture) {name=Organic Specular Base, nodeColor=<rgba 0.121 0.4041 1 1>}
  NamedRerouteDeclaration(None=MaterialFunctionCall(ScreenUV=NamedRerouteUsage()) {materialFunction=&MF_UI_ScreenSpaceUVScaling}.Scaled to Fill) {name=Normal UV Scaled to Fill, nodeColor=<rgba 0 0 0 1>}
  n45 = VectorParameter() {group=Bkgd_PurpleGlow, parameterName=Purple Glow Color, defaultValue=<rgba 0.0357 0 0.3854 1>}
  n46 = ScalarParameter() {defaultValue=1.5, group=Bkgd_PurpleGlow, parameterName=Purple Glow Scale X}
  n47 = ScalarParameter() {defaultValue=1, group=Bkgd_PurpleGlow, parameterName=Purple Glow Scale Y}
  n48 = Add(
   A=NamedRerouteUsage()
   B=AppendVector(A=Constant(), B=ScalarParameter() {defaultValue=-0.15, group=Bkgd_PurpleGlow, parameterName=Purple Glow Height Offset})
  )
  n49 = MaterialFunctionCall(UVs=@n48, Texture Scale=AppendVector(A=Multiply(B=@n46) {constA=1}, B=Multiply(B=@n47) {constA=1})) {materialFunction=&ScaleUVsByCenter}
  n50 = Constant()
  n51 = Constant() {r=1}
  n52 = OneMinus(None=MaterialFunctionCall(UV Channel=@n48) {materialFunction=&LinearGradient}.VGradient)
  n53 = Subtract(
   A=OneMinus(None=MaterialFunctionCall(UV=@n49.UVs, GradientStart=@n50, GradientEnd=@n51) {materialFunction=&MF_UI_RadialGradient}.Smooth)
   B=Saturate(None=Multiply(A=@n52) {constB=2.5})
  )
  NamedRerouteDeclaration(None=Multiply(A=LinearInterpolate(B=@n45.RGB, Alpha=@n53), B=ScalarParameter() {defaultValue=0.1, group=Bkgd_PurpleGlow, parameterName=Purple Glow Alpha})) {name=Purple Glow, nodeColor=<rgba 0 0.4452 1 1>}
  n54 = Constant()
  n55 = ScalarParameter() {defaultValue=0.1, group=Bkgd_SpotLight, parameterName=Spotlight Height Position}
  n56 = Frac(None=Multiply(A=Time(), B=ScalarParameter() {defaultValue=0.1, group=Bkgd_Specular, parameterName=Slow-moving Shine dTime multiplier}))
  n57 = Constant() {r=-2}
  n58 = Constant() {r=2}
  n59 = MaterialFunctionCall(
   UVs=Add(A=NamedRerouteUsage(), B=AppendVector(A=@n54, B=@n55))
   Size=Add(
    A=ScalarParameter() {defaultValue=1, group=Bkgd_Specular, parameterName=Slow-moving Shine Starting Size}
    B=MaterialFunctionCall(Input=@n56, Target Low=@n57, Target High=@n58) {materialFunction=&RemapValueRange}.Result
   )
   OutlineThickness=Multiply(A=Constant() {r=0.01}) {constB=5}
   GlowMin=Constant() {r=0.15}
   GlowMax=Constant() {r=-0.5}
  ) {materialFunction=&MF_UI_Circle}
  NamedRerouteDeclaration(None=OneMinus(None=Saturate(None=@n59.SDFOutline))) {name=Slow-moving Shine, nodeColor=<rgba 0.0124 0.3792 1 1>}
  n60 = Constant() {r=0.8}
  n61 = Constant() {r=1}
  n62 = Constant() {r=1.2}
  NamedRerouteDeclaration(None=MaterialFunctionCall(Input=NamedRerouteUsage(), Target Low=Multiply(A=@n60, B=@n61), Target High=Multiply(A=@n61, B=@n62)) {materialFunction=&RemapValueRange}.Result) {name=Spotlight 'Intensity' Anim Range, nodeColor=<rgba 0 0 0 1>}
  n63 = Constant() {r=0.95}
  n64 = Constant() {r=1}
  n65 = Constant() {r=1.05}
  NamedRerouteDeclaration(None=MaterialFunctionCall(Input=NamedRerouteUsage(), Target Low=Multiply(A=@n63, B=@n64), Target High=Multiply(A=@n64, B=@n65)) {materialFunction=&RemapValueRange}.Result) {name=Spotlight 'Size' Anim Range, nodeColor=<rgba 0 0 0 1>}
  n66 = ScalarParameter() {group=Bkgd_Specular, parameterName=Organic Specular Additive Intensity, sliderMax=3}
  n67 = NamedRerouteUsage()
  n68 = Saturate(None=Add(A=Multiply(A=@n67, B=@n66), B=@n67))
  n69 = Add(A=NamedRerouteUsage(), B=@n68)
  n70 = MaterialFunctionCall(Float3=MaterialFunctionCall!(Base=@n69, Blend=@n68) {materialFunction=/Engine/Functions/Engine_MaterialFunctions03/Blends/Blend_HardLight}.Result) {materialFunction=&BreakOutFloat3Components}
  MaterialExpressionAppend4Vector_1 = MaterialXAppend4Vector!(A=@n70.R, B=@n70.R, C=@n70.R, D=Constant())
  n71 = Saturate(None=Multiply(A=@MaterialExpressionAppend4Vector_1, B=ScalarParameter() {defaultValue=0.25, group=Bkgd_Specular, parameterName=Slow-moving Glow opacity}) {constB=0.25})
  n72 = MaterialFunctionCall(Float3=MaterialFunctionCall!(Base=@n68, Blend=@n69) {materialFunction=/Engine/Functions/Engine_MaterialFunctions03/Blends/Blend_ColorBurn}.Result) {materialFunction=&BreakOutFloat3Components}
  MaterialExpressionAppend4Vector_2 = MaterialXAppend4Vector(A=@n72.R, B=@n72.R, C=@n72.R, D=Constant())
  n73 = Saturate(None=Multiply(A=@MaterialExpressionAppend4Vector_2, B=ScalarParameter() {defaultValue=0.1, group=Bkgd_Specular, parameterName=Slow-moving Burn opacity}) {constB=0.1})
  n74 = Add(A=Saturate(None=Add(A=@n71, B=@n73) {constB=0}), B=@n68)
  n75 = MaterialFunctionCall(UVs=NamedRerouteUsage(), Size=Constant() {r=1.25}, GlowMin=Constant() {r=-1}, GlowMax=Constant() {r=0.5}) {materialFunction=&MF_UI_Circle}
  n76 = Add(A=@n75.Fill, B=@n75.Fill)
  NamedRerouteDeclaration(None=LinearInterpolate(B=Saturate(None=@n74), Alpha=Saturate(None=@n76)) {desc=Circle Mask}) {name=Masked Organic Specular, nodeColor=<rgba 0.121 0.4041 1 1>}
  n77 = MaterialFunctionCall(
   UV=MaterialFunctionCall(
    UVs=NamedRerouteUsage()
    Texture Scale=ScalarParameter() {defaultValue=0.85, group=Bkgd_ColorOverlay, parameterName=Color Overlay Size, sliderMax=2, sliderMin=0.01}
   ) {materialFunction=&ScaleUVsByCenter}.UVs
   GradientStart=ScalarParameter() {group=Bkgd_ColorOverlay, parameterName=Color Overlay Grad Start}
   GradientEnd=ScalarParameter() {defaultValue=0.3, group=Bkgd_ColorOverlay, parameterName=Color Overlay Grad End}
  ) {materialFunction=&MF_UI_RadialGradient}
  n78 = ScalarParameter() {defaultValue=5, group=Bkgd_ColorOverlay, parameterName=Color Overlay Intensity}
  NamedRerouteDeclaration(None=LinearInterpolate(
    B=VectorParameter() {group=Bkgd_ColorOverlay, parameterName=Color Overlay, defaultValue=<rgba 0.0137 0.1878 0.2582 1>}.RGB
    Alpha=Multiply(A=@n77.Smooth, B=@n78) {constB=2}
   )) {name=Colored Spotlight, nodeColor=<rgba 0.0074 0.5338 1 1>}
 ]
}
